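# Translucent UI material for MistRaven HUD notification fonts, producing a tinted radial glow, purple by default.
# UVs are scaled about their center by a Scale parameter and fed into a radial gradient function with adjustable start and end.
# The gradient's smooth output is multiplied by the GlowColor parameter for the tint and also passed on untinted through a reroute.
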Material /MistRaven/HUD/Notifications/Materials/M_UI_MistRaven_NotificationFont
graph {
 properties {bAutomaticallySetUsageInEditor=false, blendMode=BLEND_Translucent, materialDomain=MD_UI, usedShadingModels=Default Lit}
 nodes [
  n1 = MaterialFunctionCall(
   UV=MaterialFunctionCall(Texture Scale=ScalarParameter() {defaultValue=1.2, parameterName=Scale}) {materialFunction=&ScaleUVsByCenter}.UVs
   GradientStart=ScalarParameter() {parameterName=GradientStart}
   GradientEnd=ScalarParameter() {defaultValue=1, parameterName=GradientEnd}
  ) {materialFunction=&MF_UI_RadialGradient}
  Multiply(A=VectorParameter() {parameterName=GlowColor, defaultValue=<rgba 0.1251 0.0166 0.2917 1>}.RGB, B=@n1.Smooth)
  Reroute(None=@n1.Smooth)
 ]
}
