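# Translucent UI material for the backing of MistRaven HUD notifications, drawing a horizontal two-color gradient.
# LeftColor and RightColor parameters blend across the raw texture U coordinate.
# A parabolic mask, 4u(1-u) on the U from MF_UI_9SliceUVs, multiplies that color so it peaks in the middle and falls to zero at the left and right edges.
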Material /MistRaven/HUD/Notifications/Materials/M_UI_MistRaven_Notification_Backing
graph {
 properties {bAutomaticallySetUsageInEditor=false, blendMode=BLEND_Translucent, materialDomain=MD_UI, usedShadingModels=Default Lit}
 nodes [
  n1 = Add(A=ComponentMask(None=TextureCoordinate()) {r=true}) {constB=0}
  n2 = ComponentMask(None=MaterialFunctionCall!(UVs) {materialFunction=/FortUILibrary/MaterialFunctions/Basics/MF_UI_9SliceUVs}.UVs) {r=true}
  n3 = Multiply(A=@n2, B=OneMinus(None=@n2))
  Multiply(
   A=LinearInterpolate(
    A=VectorParameter() {parameterName=LeftColor, defaultValue=<rgba 0.3185 0.3864 0.6654 1>}.RGB
    B=VectorParameter() {parameterName=RightColor, defaultValue=<rgba 0.1329 0.0152 0.2918 1>}.RGB
    Alpha=Saturate(None=@n1)
   )
   B=Saturate(None=Multiply(A=@n3) {constB=4})
  )
 ]
}
