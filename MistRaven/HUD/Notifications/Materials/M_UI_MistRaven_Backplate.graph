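# Translucent UI material that draws the MistRaven HUD notification backplate procedurally, with no textures.
# SDF boxes, a circle and a smooth union on 9-slice UVs form the antialiased silhouette and shadow masks.
# Named reroutes carry these masks, plus three skewed dashes and a symmetry-built star, both pulsing on time sines.
# Masks layer colors over a reflected gradient and combine into alpha, with parameters for opacity, shadow and star shape.
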Material /MistRaven/HUD/Notifications/Materials/M_UI_MistRaven_Backplate
graph {
 properties {bAutomaticallySetUsageInEditor=false, blendMode=BLEND_Translucent, materialDomain=MD_UI, usedShadingModels=Default Lit}
 nodes [
  Constant() {r=0.15}
  n1 = LinearInterpolate(
   A=Constant3Vector() {constant=<rgba 0.3515 0.4564 0.6795 1>}
   B=Constant3Vector() {constant=<rgba 0.4735 0.2086 0.8308 1>}
   Alpha=MaterialFunctionCall!(UVs, Offset, RotationCenter, RotationAngle=Constant() {r=0.25}) {materialFunction=/FortUILibrary/MaterialFunctions/Gradients/MF_UI_ReflectedGradient}.Result
  )
  n2 = Constant3Vector() {constant=<rgba 0.0474 0.0054 0.1042 1>}
  n3 = NamedRerouteUsage()
  n4 = Multiply(A=@n3, B=ScalarParameter() {defaultValue=0.5, parameterName=ShadowColorOpacity}) {constB=0.5}
  n5 = NamedRerouteUsage()
  n6 = NamedRerouteUsage()
  LinearInterpolate(
   A=LinearInterpolate(A=LinearInterpolate(A=@n1, B=@n2, Alpha=@n4), B=Constant3Vector() {constant=<rgba 0.7013 0.5974 0.831 1>}, Alpha=@n5)
   B=LinearInterpolate(A=Constant3Vector() {constant=<rgba 0.4737 0.2087 0.831 1>}, B=Constant3Vector() {constant=<rgba 1 1 1 1>}, Alpha=@n6)
   Alpha=@n6
  )
  n7 = MaterialFunctionCall!(DistanceFieldA=Subtract(A=NamedRerouteUsage(), B=NamedRerouteUsage()), DistanceFieldB=NamedRerouteUsage(), BlendAmount=Constant() {r=1.4}) {materialFunction=/FortUILibrary/MaterialFunctions/Shapes/Modifiers/MF_UI_SDF_Union}
  n8 = Multiply(
   A=MaterialFunctionCall(DistanceField=OneMinus(None=@n7.Smooth), Threshold=Constant() {r=0.4}) {materialFunction=&MF_UI_AntiAliasing}.Inner
   B=ScalarParameter() {defaultValue=0.6, parameterName=Opacity}
  ) {constB=0.6}
  n9 = ScalarParameter() {defaultValue=0.9, parameterName=ShadowOpacity}
  Max(A=Add(A=LinearInterpolate(A=@n8, B=@n9, Alpha=@n3), B=@n5), B=@n6)
  NamedRerouteDeclaration(None=MaterialFunctionCall!(UVs) {materialFunction=/FortUILibrary/MaterialFunctions/Basics/MF_UI_9SliceUVs}.UVs) {name=UV, nodeColor=<rgba 0 0 0 1>}
  n10 = OneMinus(None=MaterialFunctionCall(UVs=NamedRerouteUsage(), Size=Constant2Vector() {g=0.3, r=0.36}, CornerRoundness=Constant() {r=-0.2}) {materialFunction=&MF_UI_Box}.SDF)
  NamedRerouteDeclaration(None=SmoothStep(Value=Subtract(A=@n10) {constB=0.5}) {constMax=0.4}) {name=Shadow, nodeColor=<rgba 0 0 0 1>}
  n11 = MaterialFunctionCall(UVs=Add(A=NamedRerouteUsage(), B=Constant2Vector() {g=0.315, r=-0.25}), Texture Scale=Constant() {r=0.18}) {materialFunction=&ScaleUVsByCenter}
  n12 = Constant2Vector() {r=-45}
  n13 = MaterialFunctionCall!(UV=Add(A=@n11.UVs, B=Constant2Vector() {r=-0.3}) {constB=0}, Skew Angle=@n12) {materialFunction=/FortUILibrary/MaterialFunctions/Basics/MF_UI_Skew}
  n14 = Constant2Vector() {g=0.3, r=0.15}
  n15 = Constant() {r=0.1}
  NamedRerouteDeclaration(None=MaterialFunctionCall(DistanceField=MaterialFunctionCall!(UVs=@n13.Skewed UVs, Size=@n14, CornerRoundness=@n15, KeepAspectRatio, PixelSize) {materialFunction=/FortUILibrary/MaterialFunctions/Shapes/SDFs/MF_SDF_Box}.SDF) {materialFunction=&MF_UI_AntiAliasing}.Inner) {name=ThirdBox, nodeColor=<rgba 0 0 0 1>}
  n16 = MaterialFunctionCall(Speed=Constant() {r=0.4}) {materialFunction=&DebugTimeSine}
  n17 = MaterialFunctionCall(Speed=Constant() {r=0.5}) {materialFunction=&DebugTimeSine}
  n18 = Max(
   A=Multiply(A=NamedRerouteUsage(), B=LinearInterpolate(Alpha=@n16.Result) {constA=0.3})
   B=Multiply(A=NamedRerouteUsage(), B=LinearInterpolate(Alpha=@n17.Result) {constA=0.5})
  )
  n19 = Constant() {r=0.3}
  n20 = Multiply(
   A=NamedRerouteUsage()
   B=LinearInterpolate(Alpha=MaterialFunctionCall(Speed=@n19) {materialFunction=&DebugTimeSine}.Result) {constA=0.4}
  )
  NamedRerouteDeclaration(None=Multiply(A=Max(A=@n18, B=@n20), B=ScalarParameter() {defaultValue=0.6, parameterName=Opacity}) {constB=0.6}) {name=Dashes, nodeColor=<rgba 0 0 0 1>}
  n21 = Constant() {r=0.5}
  n22 = Constant() {r=0.6}
  n23 = MaterialFunctionCall(UVs=NamedRerouteUsage(), Size=Constant2Vector() {g=0.4, r=0.4}, CornerRoundness=Constant() {r=-0.5}) {materialFunction=&MF_UI_Box}
  NamedRerouteDeclaration(None=SmoothStep(Min=@n21, Max=@n22, Value=OneMinus(None=@n23.SDF)) {constMax=0.6, constMin=0.5}) {name=Shape, nodeColor=<rgba 0 0 0 1>}
  n24 = MaterialFunctionCall(
   UVs=MaterialFunctionCall(UVs=NamedRerouteUsage(), "Rotation Angle (0-1)"=Constant() {r=0.125}) {materialFunction=&CustomRotator}.Rotated Values
   Size=Constant() {r=0.57}
  ) {materialFunction=&MF_UI_Box}
  NamedRerouteDeclaration(None=SmoothStep(Min=@n21, Max=@n22, Value=OneMinus(None=@n24.SDF)) {constMax=0.6, constMin=0.5}) {name=Corners, nodeColor=<rgba 0 0 0 1>}
  n25 = NamedRerouteUsage()
  n26 = Constant2Vector() {g=0.46, r=-0.1}
  n27 = OneMinus(None=MaterialFunctionCall(UVs=Add(A=@n25, B=@n26), Size=Constant() {r=0.15}) {materialFunction=&MF_UI_Circle}.SDF)
  NamedRerouteDeclaration(None=SmoothStep(Value=Subtract(A=@n27) {constB=0.8}) {constMax=0.3, constMin=0.12}) {name=Curve, nodeColor=<rgba 0 0 0 1>}
  n28 = MaterialFunctionCall(UVs=Add(A=NamedRerouteUsage(), B=Constant2Vector() {g=0.42, r=-0.105}), Texture Scale=Constant() {r=0.2}) {materialFunction=&ScaleUVsByCenter}
  n29 = ConstantBiasScale(None=@n28.UVs) {bias=-0.5, scale=2}
  n30 = SmoothStep(Value=Subtract(A=Abs(None=@n29), B=Constant2Vector() {g=0.5, r=0.5}) {constB=0.5}) {constMax=0.2}
  n31 = MaterialFunctionCall!(UV=@n28.UVs, Swizzle UV) {materialFunction=/FortUILibrary/MaterialFunctions/Basics/MF_UI_Symmetry_2x2_New}
  n32 = ScalarParameter() {defaultValue=0.5, group=Star, parameterName=CircleOffset, sortPriority=0}
  n33 = MaterialFunctionCall(
   UVs=MaterialFunctionCall(UVs=Add(A=@n31.Symmetry, B=@n32), Texture Scale=Constant() {r=0.5}) {materialFunction=&ScaleUVsByCenter}.UVs
   Size=ScalarParameter() {defaultValue=5, group=Star, parameterName=Size, sortPriority=1}
  ) {materialFunction=&MF_UI_Circle}
  n34 = Multiply(
   A=OneMinus(None=Length(None=@n30) {is=class MaterialExpressionLength})
   B=MaterialFunctionCall(DistanceField=OneMinus(None=@n33.SDF), Threshold=Constant() {r=1.1}) {materialFunction=&MF_UI_AntiAliasing}.Inner
  )
  n35 = OneMinus(None=MaterialFunctionCall(UVs=@n28.UVs, Size=Constant()) {materialFunction=&MF_UI_Circle}.SDF)
  n36 = MaterialFunctionCall(Speed=Constant() {r=0.3}) {materialFunction=&DebugTimeSine}
  n37 = Max(A=@n34, B=Multiply(A=Power(Base=@n35), B=LinearInterpolate(Alpha=@n36.Result) {constB=3}) {constB=2})
  NamedRerouteDeclaration(None=Saturate(None=LinearInterpolate(A=@n34, B=@n37) {constAlpha=0.4})) {name=Star, nodeColor=<rgba 0 0 0 1>}
  n38 = MaterialFunctionCall(UV=Add(A=@n11.UVs, B=Constant2Vector() {r=0.3}) {constB=0}, Skew Angle=@n12) {materialFunction=&MF_UI_Skew}
  NamedRerouteDeclaration(None=MaterialFunctionCall(DistanceField=MaterialFunctionCall(UVs=@n38.Skewed UVs, Size=@n14, CornerRoundness=@n15) {materialFunction=&MF_SDF_Box}.SDF) {materialFunction=&MF_UI_AntiAliasing}.Inner) {name=FirstBox, nodeColor=<rgba 0 0 0 1>}
  n39 = MaterialFunctionCall(UV=Add(A=@n11.UVs, B=Constant2Vector()) {constB=0}, Skew Angle=@n12) {materialFunction=&MF_UI_Skew}
  NamedRerouteDeclaration(None=MaterialFunctionCall(DistanceField=MaterialFunctionCall(UVs=@n39.Skewed UVs, Size=@n14, CornerRoundness=@n15) {materialFunction=&MF_SDF_Box}.SDF) {materialFunction=&MF_UI_AntiAliasing}.Inner) {name=SecondBox, nodeColor=<rgba 0 0 0 1>}
 ]
}
