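# Translucent UI material that draws the decorative emblem for MistRaven HUD notifications: scan lines, a star and a segmented outer ring.
# The ring subtracts four radial-coordinate SDF boxes from a base circle, each rotating at its own rate from material time or TimeScalar.
# An MSDF texture supplies the star, while a sine pattern masked by a rotated box and radial gradient forms the scan lines.
# AbilityReady adjusts layer opacities, scaled by MasterOpacity.
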Material /MistRaven/HUD/Notifications/Materials/M_UI_MistRaven_NotificationDecor
graph {
 properties {bAutomaticallySetUsageInEditor=false, blendMode=BLEND_Translucent, materialDomain=MD_UI, usedShadingModels=Default Lit}
 nodes [
  n1 = NamedRerouteUsage()
  LinearInterpolate(
   A=VectorParameter() {defaultValue=<rgba 1 1 1 1>}.RGB
   B=VectorParameter() {parameterName=ScanLineColor, defaultValue=<rgba 1 1 1 1>}.RGB
   Alpha=Saturate(None=Multiply(A=@n1) {constB=20})
  )
  n2 = Multiply(A=ScalarParameter() {defaultValue=0.85, parameterName=BackingGradientCircleSize}, B=Constant() {r=0.85})
  n3 = Add(A=OneMinus(None=ScalarParameter() {defaultValue=1.05, parameterName=BackingGradientCircleThickness}))
  n4 = ScalarParameter() {defaultValue=0.85, parameterName=BackingGradientCircleSize}
  n5 = Constant2Vector() {g=0.2, r=0.1}
  n6 = MaterialFunctionCall(
   UV=MaterialFunctionCall(Texture Scale=Multiply(A=@n3, B=@n4)) {materialFunction=&ScaleUVsByCenter}.UVs
   GradientStart=@n5.R
   GradientEnd=@n5.R
  ) {materialFunction=&MF_UI_RadialGradient}
  Multiply(
   A=MaterialFunctionCall(DistanceField=MaterialFunctionCall!(UVs, Radius=@n2) {materialFunction=/FortUILibrary/MaterialFunctions/Shapes/SDFs/MF_SDF_Circle}.SDF) {materialFunction=&MF_UI_AntiAliasing}.Inner
   B=OneMinus(None=@n6.Smooth)
  )
  n7 = MaterialFunctionCall(
   UVs=MaterialFunctionCall(UVs=TextureCoordinate(), Texture Scale=ScalarParameter() {defaultValue=0.8, parameterName=ScanLineMaskSize}) {materialFunction=&ScaleUVsByCenter}.UVs
   "Rotation Angle (0-1)"=Constant() {r=0.125}
  ) {materialFunction=&CustomRotator}
  n8 = Multiply(A=ScalarParameter() {defaultValue=0.8, parameterName=ScanLineMaskSize}, B=Constant() {r=1})
  n9 = Multiply(
   A=MaterialFunctionCall(
    DistanceField=MaterialFunctionCall!(UVs=@n7.Rotated Values, Size, CornerRoundness, KeepAspectRatio, PixelSize) {materialFunction=/FortUILibrary/MaterialFunctions/Shapes/SDFs/MF_SDF_Box}.SDF
    Smoothing=Constant() {r=0.5}
   ) {materialFunction=&MF_UI_AntiAliasing}.Inner
   B=MaterialFunctionCall(UV=MaterialFunctionCall(Texture Scale=@n8) {materialFunction=&ScaleUVsByCenter}.UVs) {materialFunction=&MF_UI_RadialGradient}.Smooth
  ) {constB=0.5}
  Multiply(A=@n9, B=NamedRerouteUsage())
  n10 = Multiply(
   A=NamedRerouteUsage()
   B=LinearInterpolate(B=ScalarParameter() {defaultValue=0.65, parameterName=ScanlineOpacity}, Alpha=ScalarParameter() {defaultValue=1, parameterName=AbilityReady}) {constA=0.2}
  )
  n11 = Multiply(
   A=NamedRerouteUsage()
   B=LinearInterpolate(
    B=ScalarParameter() {defaultValue=0.5, parameterName=InternalStarOpacity}
    Alpha=ScalarParameter() {defaultValue=1, parameterName=AbilityReady}
   )
  )
  n12 = NamedRerouteUsage()
  n13 = LinearInterpolate(
   B=ScalarParameter() {defaultValue=0.7, parameterName=OuterCircleOpacity}
   Alpha=ScalarParameter() {defaultValue=1, parameterName=AbilityReady}
  ) {constA=0.5}
  Multiply(A=Max(A=LinearInterpolate(A=@n10, B=@n11), B=Multiply(A=@n12, B=@n13)), B=ScalarParameter() {defaultValue=1, parameterName=MasterOpacity})
  n14 = Constant() {r=0.85}
  n15 = ScalarParameter() {defaultValue=1, parameterName=OuterRingScale}
  n16 = NamedRerouteUsage()
  n17 = MaterialFunctionCall(
   UVs=MaterialFunctionCall(UVs=TextureCoordinate(), Texture Scale=Multiply(A=@n14, B=@n15)) {materialFunction=&ScaleUVsByCenter}.UVs
   "Rotation Angle (0-1)"=Multiply(A=ScalarParameter() {defaultValue=0.928, parameterName=SDFBoxCutout1Rotation}, B=Multiply(A=@n16) {constB=0.1})
  ) {materialFunction=&CustomRotator}
  n18 = Constant2Vector() {g=0.1, r=0.1}
  n19 = Multiply(A=@n18.G, B=ScalarParameter() {defaultValue=1, parameterName=OuterRingThickness})
  NamedRerouteDeclaration(None=MaterialFunctionCall(
    UVs=MaterialFunctionCall!(Vector or UVs=@n17.Rotated Values) {materialFunction=/FortUILibrary/MaterialFunctions/Basics/MF_UI_CheaperVectorToRadial}.Radial Coordinates
    Size=AppendVector(A=@n18.R, B=@n19)
   ) {materialFunction=&MF_SDF_Box}.SDF) {name=SDFBoxCutout1, nodeColor=<rgba 0.1953 1 0 1>}
  n20 = Constant() {r=0.85}
  n21 = ScalarParameter() {defaultValue=1, parameterName=OuterRingScale}
  n22 = MaterialFunctionCall(UVs=MaterialFunctionCall(UVs=TextureCoordinate(), Texture Scale=Multiply(A=@n20, B=@n21)) {materialFunction=&ScaleUVsByCenter}.UVs) {materialFunction=&CustomRotator}
  n23 = Constant2Vector() {g=0.05, r=1.1}
  n24 = Multiply(A=@n23.G, B=ScalarParameter() {defaultValue=1, parameterName=OuterRingThickness})
  NamedRerouteDeclaration(None=MaterialFunctionCall(
    UVs=MaterialFunctionCall(Vector or UVs=@n22.Rotated Values) {materialFunction=&MF_UI_CheaperVectorToRadial}.Radial Coordinates
    Size=AppendVector(A=@n23.R, B=@n24)
   ) {materialFunction=&MF_SDF_Box}.SDF) {name=BaseCircle, nodeColor=<rgba 1 0 0.5156 1>}
  n25 = TextureSample(UVs=MaterialFunctionCall(Texture Scale=ScalarParameter() {defaultValue=1.25, parameterName=StarScale}) {materialFunction=&ScaleUVsByCenter}.UVs) {texture=/MistRaven/HUD/Notifications/Textures/T_UI_MSDF_MistRaven_Star}
  NamedRerouteDeclaration(None=MaterialFunctionCall(
    DistanceField=MaterialFunctionCall!(MSDF RGB=@n25.RGB) {materialFunction=/FortUILibrary/MaterialFunctions/Shapes/Modifiers/MF_UI_MSDFtoSDF}.SDF
    Threshold=Constant() {r=0.2}
   ) {materialFunction=&MF_UI_AntiAliasing}.Outer) {name=InternalStarOpacity, nodeColor=<rgba 1 0 0.1172 1>}
  NamedRerouteDeclaration(None=@n6.Smooth) {name=RadialMask, nodeColor=<rgba 0.9219 0 1 1>}
  n26 = LinearInterpolate(A=ScalarParameter() {parameterName=TimeScalar}, B=Time(), Alpha=ScalarParameter() {defaultValue=1, parameterName=IsMaterialTime})
  n27 = ScalarParameter() {defaultValue=0.5, parameterName=RotationSpeed}
  NamedRerouteDeclaration(None=Add(A=Multiply(A=@n26, B=@n27), B=ScalarParameter() {parameterName=StartinAnimSeed})) {name=TimeSpeed, nodeColor=<rgba 1 0.6797 0 1>}
  n28 = Constant() {r=0.85}
  n29 = ScalarParameter() {defaultValue=1, parameterName=OuterRingScale}
  n30 = NamedRerouteUsage()
  n31 = MaterialFunctionCall(
   UVs=MaterialFunctionCall(UVs=TextureCoordinate(), Texture Scale=Multiply(A=@n28, B=@n29)) {materialFunction=&ScaleUVsByCenter}.UVs
   "Rotation Angle (0-1)"=Multiply(A=ScalarParameter() {defaultValue=1.2, parameterName=SDFBoxCutout2Rotation}, B=Multiply(A=@n30) {constB=0.05})
  ) {materialFunction=&CustomRotator}
  n32 = Constant2Vector() {g=0.1, r=0.075}
  n33 = Multiply(A=@n32.G, B=ScalarParameter() {defaultValue=1, parameterName=OuterRingThickness})
  NamedRerouteDeclaration(None=MaterialFunctionCall(
    UVs=MaterialFunctionCall(Vector or UVs=@n31.Rotated Values) {materialFunction=&MF_UI_CheaperVectorToRadial}.Radial Coordinates
    Size=AppendVector(A=@n32.R, B=@n33)
   ) {materialFunction=&MF_SDF_Box}.SDF) {name=SDFBoxCutout2, nodeColor=<rgba 0 0.8359 1 1>}
  n34 = ScalarParameter() {defaultValue=0.9, parameterName=TrimScale}
  n35 = ScalarParameter() {defaultValue=1, parameterName=OuterRingScale}
  n36 = NamedRerouteUsage()
  n37 = MaterialFunctionCall(
   UVs=MaterialFunctionCall(UVs=TextureCoordinate(), Texture Scale=Multiply(A=@n34, B=@n35)) {materialFunction=&ScaleUVsByCenter}.UVs
   "Rotation Angle (0-1)"=Multiply(A=ScalarParameter() {defaultValue=-0.544, parameterName=SDFBoxLimiter1Rotation}, B=Multiply(A=@n36) {constB=0.25})
  ) {materialFunction=&CustomRotator}
  n38 = Constant2Vector() {g=0.05, r=0.175}
  n39 = Multiply(A=@n38.G, B=ScalarParameter() {defaultValue=1, parameterName=OuterRingThickness})
  NamedRerouteDeclaration(None=MaterialFunctionCall(
    UVs=MaterialFunctionCall(Vector or UVs=@n37.Rotated Values) {materialFunction=&MF_UI_CheaperVectorToRadial}.Radial Coordinates
    Size=AppendVector(A=@n38.R, B=@n39)
   ) {materialFunction=&MF_SDF_Box}.SDF) {name=SDFBoxLimiter1, nodeColor=<rgba 1 0.8438 0 1>}
  n40 = ScalarParameter() {defaultValue=0.9, parameterName=TrimScale}
  n41 = ScalarParameter() {defaultValue=1, parameterName=OuterRingScale}
  n42 = NamedRerouteUsage()
  n43 = MaterialFunctionCall(
   UVs=MaterialFunctionCall(UVs=TextureCoordinate(), Texture Scale=Multiply(A=@n40, B=@n41)) {materialFunction=&ScaleUVsByCenter}.UVs
   "Rotation Angle (0-1)"=Multiply(A=ScalarParameter() {defaultValue=0.072, parameterName=SDFBoxLimiter2Rotation}, B=Multiply(A=@n42) {constB=-0.5})
  ) {materialFunction=&CustomRotator}
  n44 = Constant2Vector() {g=0.05, r=0.2}
  n45 = Multiply(A=@n44.G, B=ScalarParameter() {defaultValue=1, parameterName=OuterRingThickness})
  NamedRerouteDeclaration(None=MaterialFunctionCall(
    UVs=MaterialFunctionCall(Vector or UVs=@n43.Rotated Values) {materialFunction=&MF_UI_CheaperVectorToRadial}.Radial Coordinates
    Size=AppendVector(A=@n44.R, B=@n45)
   ) {materialFunction=&MF_SDF_Box}.SDF) {name=SDFBoxLimiter2, nodeColor=<rgba 0.6641 0 1 1>}
  n46 = NamedRerouteUsage()
  n47 = NamedRerouteUsage()
  n48 = MaterialFunctionCall!(
   DistanceFieldA=MaterialFunctionCall(
    DistanceFieldA=MaterialFunctionCall(DistanceFieldA=@n46, DistanceFieldB=@n47) {materialFunction=/FortUILibrary/MaterialFunctions/Shapes/Modifiers/MF_UI_SDF_Subtraction}.Exact
    DistanceFieldB=NamedRerouteUsage()
   ) {materialFunction=&MF_UI_SDF_Subtraction}.Exact
   DistanceFieldB=NamedRerouteUsage()
   BlendAmount
  ) {materialFunction=&MF_UI_SDF_Subtraction}
  n49 = NamedRerouteUsage()
  NamedRerouteDeclaration(None=MaterialFunctionCall(DistanceField=MaterialFunctionCall(DistanceFieldA=@n48.Exact, DistanceFieldB=@n49) {materialFunction=&MF_UI_SDF_Subtraction}.Exact) {materialFunction=&MF_UI_AntiAliasing}.Inner) {name=OuterCircle, nodeColor=<rgba 0 0.2969 1 1>}
  n50 = Multiply(A=TextureCoordinate(), B=ScalarParameter() {defaultValue=30, parameterName=ScanLinesAmount}) {constB=30}
  n51 = ComponentMask(None=Power(Base=Sine(None=@n50)) {constExponent=0.1}) {g=true}
  NamedRerouteDeclaration(None=Multiply(A=Multiply(A=@n51, B=@n9), B=ScalarParameter() {defaultValue=0.65, parameterName=ScanLineOpacity})) {name=ScanLinesOpacity, nodeColor=<rgba 0.5937 1 0 1>}
 ]
}
