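# Translucent UI material that draws a scan-line pattern for MistRaven HUD notifications.
# A sine of tiled UVs forms horizontal stripes, with AddedThickness widening the lines.
# The stripes are multiplied by a horizontal edge falloff from 9-slice UVs and a gradient mask built from mirrored UVs.
# Parameters rotate and scale that gradient mask, and MasterOpacity scales the combined result.
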Material /MistRaven/HUD/Notifications/Materials/M_UI_MistRaven_Notification_ScanLine
graph {
 properties {bAutomaticallySetUsageInEditor=false, blendMode=BLEND_Translucent, materialDomain=MD_UI, usedShadingModels=Default Lit}
 nodes [
  n1 = Multiply(A=TextureCoordinate()) {constB=20}
  n2 = Add(A=ComponentMask(None=Sine(None=@n1)) {g=true}, B=ScalarParameter() {defaultValue=0.5, parameterName=AddedThickness})
  n3 = ComponentMask(None=MaterialFunctionCall!(UVs) {materialFunction=/FortUILibrary/MaterialFunctions/Basics/MF_UI_9SliceUVs}.UVs) {r=true}
  n4 = Multiply(A=@n3, B=OneMinus(None=@n3))
  n5 = Saturate(None=Add(A=Multiply(A=@n4)) {constB=0})
  n6 = MaterialFunctionCall(
   UVs=MaterialFunctionCall!(UV, Swizzle UV) {materialFunction=/FortUILibrary/MaterialFunctions/Basics/MF_UI_Symmetry_2x2_New}.Symmetry
   "Rotation Angle (0-1)"=ScalarParameter() {defaultValue=0.4, parameterName=MaskRotation}
  ) {materialFunction=&CustomRotator}
  n7 = ScalarParameter() {defaultValue=1, parameterName=MaskScale}
  n8 = OneMinus(None=ComponentMask(None=MaterialFunctionCall(UVs=@n6.Rotated Values, Texture Scale=@n7) {materialFunction=&ScaleUVsByCenter}.UVs) {r=true})
  Multiply(A=Multiply(A=Saturate(None=@n2), B=Multiply(A=@n5, B=@n8)), B=ScalarParameter() {defaultValue=0.5, parameterName=MasterOpacity})
 ]
}
