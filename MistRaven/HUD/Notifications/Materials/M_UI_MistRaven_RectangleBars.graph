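# Translucent UI material for MistRaven HUD notifications that draws five slanted rectangular bars with animated heights.
# Each bar is an antialiased SDF box on skewed, offset UVs, with GroupOffsetX, Spacing and Skew parameters controlling layout.
# A time-driven sine at a different speed per bar sets a vertical Step cutoff, so the bars rise and fall out of sync.
# Max nodes merge the five masks, which MasterOpacity then scales.
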Material /MistRaven/HUD/Notifications/Materials/M_UI_MistRaven_RectangleBars
graph {
 properties {bAutomaticallySetUsageInEditor=false, blendMode=BLEND_Translucent, materialDomain=MD_UI, usedShadingModels=Default Lit}
 nodes [
  n1 = Max(A=Max(A=NamedRerouteUsage(), B=NamedRerouteUsage()), B=NamedRerouteUsage())
  n2 = NamedRerouteUsage()
  Multiply(A=Max(A=Max(A=@n1, B=@n2), B=NamedRerouteUsage()), B=ScalarParameter() {defaultValue=0.6, parameterName=MasterOpacity})
  MaterialExpressionScalarParameter_4 = ScalarParameter() {defaultValue=0.25, parameterName=GroupOffsetX}
  n3 = Convert(None=@MaterialExpressionScalarParameter_4) {
   convertInputs [{
     type=Scalar
     expressionInput {
      expression=&self:MaterialExpressionScalarParameter_4
      inputName=None
      mask=0
      maskA=0
      maskB=0
      maskG=0
      maskR=0
      outputIndex=0
     }
    }]
   convertOutputs [{type=Vector2}]
  }
  NamedRerouteDeclaration(None=@n3) {name=GroupOffsetX, nodeColor=<rgba 1 0 0.3984 1>}
  MaterialExpressionScalarParameter_1 = ScalarParameter() {defaultValue=-20, parameterName=Skew}
  n4 = Convert(None=@MaterialExpressionScalarParameter_1) {
   convertInputs [{
     type=Scalar
     expressionInput {
      expression=&self:MaterialExpressionScalarParameter_1
      inputName=None
      mask=0
      maskA=0
      maskB=0
      maskG=0
      maskR=0
      outputIndex=0
     }
    }]
   convertOutputs [{type=Vector2}]
  }
  NamedRerouteDeclaration(None=@n4) {name=Skew, nodeColor=<rgba 0 0.7187 1 1>}
  MaterialExpressionScalarParameter_5 = ScalarParameter() {defaultValue=0.15, parameterName=Spacing}
  n5 = Convert(None=@MaterialExpressionScalarParameter_5) {
   convertInputs [{
     type=Scalar
     expressionInput {
      expression=&self:MaterialExpressionScalarParameter_5
      inputName=None
      mask=0
      maskA=0
      maskB=0
      maskG=0
      maskR=0
      outputIndex=0
     }
    }]
   convertOutputs [{type=Vector2}]
  }
  NamedRerouteDeclaration(None=@n5) {name=Spacing, nodeColor=<rgba 1 0.9609 0 1>}
  n6 = TextureCoordinate()
  n7 = Add(A=NamedRerouteUsage(), B=Constant2Vector() {r=-0.15}) {constB=2}
  n8 = MaterialFunctionCall!(
   UVs=MaterialFunctionCall!(UV=Add(A=@n6, B=@n7), Skew Angle=NamedRerouteUsage()) {materialFunction=/FortUILibrary/MaterialFunctions/Basics/MF_UI_Skew}.Skewed UVs
   Size=Constant2Vector() {g=0.2, r=0.1}
   CornerRoundness
   KeepAspectRatio
   PixelSize
  ) {materialFunction=/FortUILibrary/MaterialFunctions/Shapes/SDFs/MF_SDF_Box}
  n9 = OneMinus(None=ComponentMask(None=TextureCoordinate()) {g=true})
  n10 = Multiply(A=Time()) {constB=0.5}
  n11 = Constant2Vector() {g=1}
  n12 = Constant2Vector() {g=0.6, r=0.5}
  n13 = MaterialFunctionCall(Input=ConstantBiasScale(None=Sine(None=@n10)), Input Low=@n11.R, Input High=@n11.R, Target Low=@n12.R, Target High=@n12.R) {materialFunction=&RemapValueRange}
  NamedRerouteDeclaration(None=Multiply(
    A=MaterialFunctionCall(DistanceField=@n8.SDF) {materialFunction=&MF_UI_AntiAliasing}.Inner
    B=Step(Y=@n9, X=@n13.Result)
   )) {name=Rect3, nodeColor=<rgba 0.6406 0 1 1>}
  NamedRerouteDeclaration(None=Constant2Vector() {g=0.2, r=0.1}) {name=BoxSize, nodeColor=<rgba 0 1 0.3203 1>}
  n14 = TextureCoordinate()
  n15 = Add(A=NamedRerouteUsage(), B=NamedRerouteUsage())
  n16 = MaterialFunctionCall(
   UVs=MaterialFunctionCall(UV=Add(A=@n14, B=@n15), Skew Angle=NamedRerouteUsage()) {materialFunction=&MF_UI_Skew}.Skewed UVs
   Size=NamedRerouteUsage()
  ) {materialFunction=&MF_SDF_Box}
  n17 = OneMinus(None=ComponentMask(None=TextureCoordinate()) {g=true})
  n18 = Multiply(A=Time()) {constB=0.2}
  n19 = Constant2Vector() {g=1}
  n20 = Constant2Vector() {g=0.6, r=0.45}
  n21 = MaterialFunctionCall(Input=ConstantBiasScale(None=Sine(None=@n18)), Input Low=@n19.R, Input High=@n19.R, Target Low=@n20.R, Target High=@n20.R) {materialFunction=&RemapValueRange}
  NamedRerouteDeclaration(None=Multiply(
    A=MaterialFunctionCall(DistanceField=@n16.SDF) {materialFunction=&MF_UI_AntiAliasing}.Inner
    B=Step(Y=@n17, X=@n21.Result)
   )) {name=Rect2, nodeColor=<rgba 1 0 0.2344 1>}
  n22 = TextureCoordinate()
  n23 = NamedRerouteUsage()
  n24 = MaterialFunctionCall(
   UVs=MaterialFunctionCall(UV=Add(A=@n22, B=@n23), Skew Angle=NamedRerouteUsage()) {materialFunction=&MF_UI_Skew}.Skewed UVs
   Size=NamedRerouteUsage()
  ) {materialFunction=&MF_SDF_Box}
  n25 = OneMinus(None=ComponentMask(None=TextureCoordinate()) {g=true})
  n26 = Multiply(A=Time()) {constB=0.1}
  n27 = Constant2Vector() {g=1}
  n28 = Constant2Vector() {g=0.6, r=0.45}
  n29 = MaterialFunctionCall(Input=ConstantBiasScale(None=Sine(None=@n26)), Input Low=@n27.R, Input High=@n27.R, Target Low=@n28.R, Target High=@n28.R) {materialFunction=&RemapValueRange}
  NamedRerouteDeclaration(None=Multiply(
    A=MaterialFunctionCall(DistanceField=@n24.SDF) {materialFunction=&MF_UI_AntiAliasing}.Inner
    B=Step(Y=@n25, X=@n29.Result)
   )) {name=Rect1, nodeColor=<rgba 0.7109 0 1 1>}
  n30 = TextureCoordinate()
  n31 = Add(A=NamedRerouteUsage(), B=Constant2Vector() {r=-0.3}) {constB=2}
  n32 = MaterialFunctionCall(
   UVs=MaterialFunctionCall(UV=Add(A=@n30, B=@n31), Skew Angle=NamedRerouteUsage()) {materialFunction=&MF_UI_Skew}.Skewed UVs
   Size=Constant2Vector() {g=0.2, r=0.075}
  ) {materialFunction=&MF_SDF_Box}
  n33 = OneMinus(None=ComponentMask(None=TextureCoordinate()) {g=true})
  n34 = Multiply(A=Time()) {constB=0.35}
  n35 = Constant2Vector() {g=1}
  n36 = Constant2Vector() {g=0.6, r=0.485}
  n37 = MaterialFunctionCall(Input=ConstantBiasScale(None=Sine(None=@n34)), Input Low=@n35.R, Input High=@n35.R, Target Low=@n36.R, Target High=@n36.R) {materialFunction=&RemapValueRange}
  NamedRerouteDeclaration(None=Multiply(
    A=MaterialFunctionCall(DistanceField=@n32.SDF) {materialFunction=&MF_UI_AntiAliasing}.Inner
    B=Step(Y=@n33, X=@n37.Result)
   )) {name=Rect4, nodeColor=<rgba 0 1 0.3906 1>}
  n38 = TextureCoordinate()
  n39 = Add(A=NamedRerouteUsage(), B=Constant2Vector() {r=-0.55}) {constB=2}
  n40 = MaterialFunctionCall(
   UVs=MaterialFunctionCall(UV=Add(A=@n38, B=@n39), Skew Angle=NamedRerouteUsage()) {materialFunction=&MF_UI_Skew}.Skewed UVs
   Size=Constant2Vector() {g=0.2, r=0.3}
  ) {materialFunction=&MF_SDF_Box}
  n41 = OneMinus(None=ComponentMask(None=TextureCoordinate()) {g=true})
  n42 = Multiply(A=Time()) {constB=0.15}
  n43 = Constant2Vector() {g=1}
  n44 = Constant2Vector() {g=0.6, r=0.435}
  n45 = MaterialFunctionCall(Input=ConstantBiasScale(None=Sine(None=@n42)), Input Low=@n43.R, Input High=@n43.R, Target Low=@n44.R, Target High=@n44.R) {materialFunction=&RemapValueRange}
  NamedRerouteDeclaration(None=Multiply(
    A=MaterialFunctionCall(DistanceField=@n40.SDF) {materialFunction=&MF_UI_AntiAliasing}.Inner
    B=Step(Y=@n41, X=@n45.Result)
   )) {name=Rect5, nodeColor=<rgba 1 0.0703 0 1>}
 ]
}
