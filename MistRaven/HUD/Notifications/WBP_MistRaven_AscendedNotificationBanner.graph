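# HUD banner widget announcing "MIST RAVEN KIT COMPLETE" with cape and gloves icons over material-driven backplate, scanline and glow layers.
# On Construct it collapses itself, then immediately calls OnMistRavenKitComplete, which reveals it, plays a fade-and-slide-up animation and a 2D sound.
# A three-second timer fires OnHideNotification, which plays a fade-and-scale-up exit and clears the timer, without collapsing or removing the widget.
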WidgetBlueprint /MistRaven/HUD/Notifications/WBP_MistRaven_AscendedNotificationBanner
animations [
 Anim_Show {
  binding_count=1
  length_seconds=0.3
  object=&self:Anim_Show
  sequence {bindings [Overlay_Root {
     id=#1
     possessed_class=Overlay
     tracks [
      {
       type=MovieSceneFloatTrack
       display_name=Render Opacity
       sections [{type=MovieSceneFloatSection, end_frame=6, start_frame=0, channels [None {type=MovieSceneScriptingFloatChannel, keys [[0,0],[6,1]]}]}]
      }
      {
       type=MovieScene2DTransformTrack
       sections [{
         type=MovieScene2DTransformSection
         end_frame=6
         start_frame=0
         channels [
          Translation.X {type=MovieSceneScriptingFloatChannel, default=0}
          Translation.Y {type=MovieSceneScriptingFloatChannel, keys [[0,50],[6,0]]}
          Angle {type=MovieSceneScriptingFloatChannel, default=0}
          Scale.X {type=MovieSceneScriptingFloatChannel, default=1}
          Scale.Y {type=MovieSceneScriptingFloatChannel, default=1}
          Shear.X {type=MovieSceneScriptingFloatChannel, default=0}
          Shear.Y {type=MovieSceneScriptingFloatChannel, default=0}
         ]
        }]
      }
     ]
    }]}
 }
 Anim_Hide {
  binding_count=1
  length_seconds=0.2
  object=&self:Anim_Hide
  sequence {bindings [Overlay_Root {
     id=#2
     possessed_class=Overlay
     tracks [
      {
       type=MovieSceneFloatTrack
       display_name=Render Opacity
       sections [{type=MovieSceneFloatSection, end_frame=4, start_frame=0, channels [None {type=MovieSceneScriptingFloatChannel, keys [[0,1],[4,0]]}]}]
      }
      {
       type=MovieScene2DTransformTrack
       sections [{
         type=MovieScene2DTransformSection
         end_frame=4
         start_frame=0
         channels [
          Translation.X {type=MovieSceneScriptingFloatChannel, default=0}
          Translation.Y {type=MovieSceneScriptingFloatChannel, default=0}
          Angle {type=MovieSceneScriptingFloatChannel, default=0}
          Scale.X {type=MovieSceneScriptingFloatChannel, keys [[0,1],[4,1.25]]}
          Scale.Y {type=MovieSceneScriptingFloatChannel, keys [[0,1],[4,1.25]]}
          Shear.X {type=MovieSceneScriptingFloatChannel, default=0}
          Shear.Y {type=MovieSceneScriptingFloatChannel, default=0}
         ]
        }]
      }
     ]
    }]}
 }
]
graphs [EventGraph {nodes [
   OnMistRavenKitComplete() {is=title K2Node_CustomEvent}
   > SetVisibility(InVisibility=HitTestInvisible)
   > PlayAnimationForward(InAnimation=Get Anim_Show() -> (Anim_Show:Widget Animation Object Reference))
   > n1 = PlaySound2D(Sound=/MistRaven/Assets/Audio/MSS_MistRaven_KitComplete_UI)
   n2 = OnHideNotification() {is=title K2Node_CustomEvent}
   n3 = Set Timer by Event(execute=@n1.then, Delegate=@n2, Time=3, bLooping=false)
   Event Construct()
   > SetVisibility(InVisibility=Collapsed)
   > OnMistRavenKitComplete() {is=title K2Node_CallFunction}
   Clear and Invalidate Timer by Handle(execute=@n4.then, Handle=@n3)
   n4 = PlayAnimation(execute=@n2.then, InAnimation=Get Anim_Hide() -> (Anim_Hide:Widget Animation Object Reference))
  ]}]
variables []
widgets {
 info {inheritedWidgetCount=0, parentClass=&UserWidget, rootWidgetClass=&ScaleBox, widgetCount=17}
 widgets [
  ScaleBox_0 {widgetClassPath=&ScaleBox, widget=ScaleBox_0, properties {stretch=UserSpecified, userSpecifiedScale=2}}
  Overlay_Root {widgetClassPath=&Overlay, parent=ScaleBox_0, widget=Overlay_Root, slot {class=ScaleBoxSlot}}
  Image_Background {
   widgetClassPath=&Image
   bIsVariable=true
   parent=Overlay_Root
   widget=Image_Background
   properties {
    renderOpacity=0.75
    brush {
     bIsDynamicallyLoaded=false
     drawAs=Image
     imageType=FullColor
     mirroring=NoMirror
     resourceName=None
     resourceObject=/MistRaven/HUD/Notifications/Materials/M_UI_MistRaven_Backplate
     tiling=NoTile
     imageSize=<32 32>
     margin=<ltrb 0 0 0 0>
     outlineSettings {
      bUseBrushTransparency=false
      roundingType=HalfHeightRadius
      width=0
      color {colorUseRule=UseColor_Specified, specifiedColor=<rgba 0 0 0 0>}
      cornerRadii=<0 0 0 0>
     }
     tintColor {colorUseRule=UseColor_Specified, specifiedColor=<rgba 1 1 1 1>}
     uVRegion {bIsValid=false, max=<0 0>, min=<0 0>}
    }
   }
   slot {class=OverlaySlot, properties {horizontalAlignment=HAlign_Fill, verticalAlignment=VAlign_Fill, padding=<ltrb 0 5 0 5>}}
  }
  GridPanel_77 {
   widgetClassPath=&GridPanel
   parent=Overlay_Root
   widget=GridPanel_77
   slot {class=OverlaySlot, properties {horizontalAlignment=HAlign_Center, verticalAlignment=VAlign_Center}}
  }
  Overlay_Background {
   widgetClassPath=&Overlay
   parent=GridPanel_77
   widget=Overlay_Background
   slot {class=GridSlot, properties {columnSpan=3, padding=<ltrb 0 10 0 10>}}
  }
  Overlay_TopRight {
   widgetClassPath=&Overlay
   parent=Overlay_Background
   widget=Overlay_TopRight
   properties {renderOpacity=0.4}
   slot {class=OverlaySlot, properties {padding=<ltrb 0 -12 0 0>}}
  }
  Image_Rectangles {
   widgetClassPath=&Image
   bIsVariable=true
   parent=Overlay_TopRight
   widget=Image_Rectangles
   properties {
    brush {
     bIsDynamicallyLoaded=false
     drawAs=Image
     imageType=FullColor
     mirroring=NoMirror
     resourceName=None
     resourceObject=/MistRaven/HUD/Notifications/Materials/M_UI_MistRaven_RectangleBars
     tiling=NoTile
     imageSize=<24 24>
     margin=<ltrb 0 0 0 0>
     outlineSettings {
      bUseBrushTransparency=false
      roundingType=HalfHeightRadius
      width=0
      color {colorUseRule=UseColor_Specified, specifiedColor=<rgba 0 0 0 0>}
      cornerRadii=<0 0 0 0>
     }
     tintColor {colorUseRule=UseColor_Specified, specifiedColor=<rgba 1 1 1 1>}
     uVRegion {bIsValid=false, max=<0 0>, min=<0 0>}
    }
    renderTransform {angle=0, scale=<1 1>, shear=<0 0>, translation=<60 -4>}
   }
   slot {class=OverlaySlot}
  }
  Image_ScanLines {
   widgetClassPath=&Image
   bIsVariable=true
   parent=Overlay_Background
   widget=Image_ScanLines
   properties {
    renderOpacity=0.4
    brush {
     bIsDynamicallyLoaded=false
     drawAs=Image
     imageType=FullColor
     mirroring=NoMirror
     resourceName=None
     resourceObject=/MistRaven/HUD/Notifications/Materials/M_UI_MistRaven_Notification_ScanLine
     tiling=NoTile
     imageSize=<32 32>
     margin=<ltrb 0 0 0 0>
     outlineSettings {
      bUseBrushTransparency=false
      roundingType=HalfHeightRadius
      width=0
      color {colorUseRule=UseColor_Specified, specifiedColor=<rgba 0 0 0 0>}
      cornerRadii=<0 0 0 0>
     }
     tintColor {colorUseRule=UseColor_Specified, specifiedColor=<rgba 1 1 1 1>}
     uVRegion {bIsValid=false, max=<0 0>, min=<0 0>}
    }
   }
   slot {class=OverlaySlot, properties {horizontalAlignment=HAlign_Fill, verticalAlignment=VAlign_Fill, padding=<ltrb 30 0 0 0>}}
  }
  Overlay_ItemIcon {
   widgetClassPath=&Overlay
   parent=GridPanel_77
   widget=Overlay_ItemIcon
   slot {class=GridSlot, properties {horizontalAlignment=HAlign_Center, verticalAlignment=VAlign_Center}}
  }
  Image_ItemBackground {
   widgetClassPath=&Image
   bIsVariable=true
   parent=Overlay_ItemIcon
   widget=Image_ItemBackground
   properties {brush {
     bIsDynamicallyLoaded=false
     drawAs=Box
     imageType=FullColor
     mirroring=NoMirror
     resourceName=None
     resourceObject=/MistRaven/HUD/Notifications/Materials/M_UI_MistRaven_NotificationDecor
     tiling=NoTile
     imageSize=<64 64>
     margin=<ltrb 0.5 0.5 0.5 0.5>
     outlineSettings {
      bUseBrushTransparency=false
      roundingType=HalfHeightRadius
      width=0
      color {colorUseRule=UseColor_Specified, specifiedColor=<rgba 0 0 0 0>}
      cornerRadii=<0 0 0 0>
     }
     tintColor {colorUseRule=UseColor_Specified, specifiedColor=<rgba 1 1 1 1>}
     uVRegion {bIsValid=false, max=<0 0>, min=<0 0>}
    }}
   slot {class=OverlaySlot, properties {horizontalAlignment=HAlign_Center, verticalAlignment=VAlign_Center}}
  }
  Image_Circle {
   widgetClassPath=&Image
   bIsVariable=true
   parent=Overlay_ItemIcon
   widget=Image_Circle
   properties {brush {
     bIsDynamicallyLoaded=false
     drawAs=Image
     imageType=FullColor
     mirroring=NoMirror
     resourceName=None
     resourceObject=/MistRaven/HUD/Notifications/Materials/MI_UI_MistRaven_ItemBackingCircle
     tiling=NoTile
     imageSize=<48 48>
     margin=<ltrb 0 0 0 0>
     outlineSettings {
      bUseBrushTransparency=false
      roundingType=HalfHeightRadius
      width=0
      color {colorUseRule=UseColor_Specified, specifiedColor=<rgba 0 0 0 0>}
      cornerRadii=<0 0 0 0>
     }
     tintColor {colorUseRule=UseColor_Specified, specifiedColor=<rgba 1 1 1 1>}
     uVRegion {bIsValid=false, max=<0 0>, min=<0 0>}
    }}
   slot {class=OverlaySlot, properties {horizontalAlignment=HAlign_Center, verticalAlignment=VAlign_Center}}
  }
  Image_Glow {
   widgetClassPath=&Image
   bIsVariable=true
   parent=Overlay_ItemIcon
   widget=Image_Glow
   properties {
    brush {
     bIsDynamicallyLoaded=false
     drawAs=Image
     imageType=FullColor
     mirroring=NoMirror
     resourceName=None
     resourceObject=/MistRaven/HUD/Notifications/Materials/MI_UI_MistRaven_ItemBackingGlow
     tiling=NoTile
     imageSize=<37 48>
     margin=<ltrb 0 0 0 0>
     outlineSettings {
      bUseBrushTransparency=false
      roundingType=HalfHeightRadius
      width=0
      color {colorUseRule=UseColor_Specified, specifiedColor=<rgba 0 0 0 0>}
      cornerRadii=<0 0 0 0>
     }
     tintColor {colorUseRule=UseColor_Specified, specifiedColor=<rgba 1 1 1 1>}
     uVRegion {bIsValid=false, max=<0 0>, min=<0 0>}
    }
    renderTransform {angle=0, scale=<2 2>, shear=<0 0>, translation=<0 0>}
   }
   slot {class=OverlaySlot, properties {horizontalAlignment=HAlign_Center, verticalAlignment=VAlign_Center}}
  }
  Overlay_0 {
   widgetClassPath=&Overlay
   parent=Overlay_ItemIcon
   widget=Overlay_0
   properties {renderTransform {angle=0, scale=<1 1>, shear=<0 0>, translation=<0 3>}}
   slot {class=OverlaySlot, properties {horizontalAlignment=HAlign_Center, verticalAlignment=VAlign_Center}}
  }
  Image_ItemIcon_Cape {
   widgetClassPath=&Image
   bIsVariable=true
   parent=Overlay_0
   widget=Image_ItemIcon_Cape
   properties {
    brush {
     bIsDynamicallyLoaded=false
     drawAs=Box
     imageType=FullColor
     mirroring=NoMirror
     resourceName=None
     resourceObject=/MistRaven/UI/Icons/T_IconWide_Combined_MistRaven
     tiling=NoTile
     imageSize=<54 27>
     margin=<ltrb 0.5 0 0 0>
     outlineSettings {
      bUseBrushTransparency=false
      roundingType=HalfHeightRadius
      width=0
      color {colorUseRule=UseColor_Specified, specifiedColor=<rgba 0 0 0 0>}
      cornerRadii=<0 0 0 0>
     }
     tintColor {colorUseRule=UseColor_Specified, specifiedColor=<rgba 1 1 1 1>}
     uVRegion {bIsValid=false, max=<0 0>, min=<0 0>}
    }
    renderTransform {angle=0, scale=<1 1>, shear=<0 0>, translation=<5 -9>}
   }
   slot {class=OverlaySlot}
  }
  Image_ItemIcon_Gloves {
   widgetClassPath=&Image
   bIsVariable=true
   parent=Overlay_0
   widget=Image_ItemIcon_Gloves
   properties {
    brush {
     bIsDynamicallyLoaded=false
     drawAs=Box
     imageType=FullColor
     mirroring=NoMirror
     resourceName=None
     resourceObject=&T_IconWide_Combined_MistRaven
     tiling=NoTile
     imageSize=<54 27>
     margin=<ltrb 0 0 0.5 0>
     outlineSettings {
      bUseBrushTransparency=false
      roundingType=HalfHeightRadius
      width=0
      color {colorUseRule=UseColor_Specified, specifiedColor=<rgba 0 0 0 0>}
      cornerRadii=<0 0 0 0>
     }
     tintColor {colorUseRule=UseColor_Specified, specifiedColor=<rgba 1 1 1 1>}
     uVRegion {bIsValid=false, max=<0 0>, min=<0 0>}
    }
    renderTransform {angle=0, scale=<1 1>, shear=<0 0>, translation=<-8 0>}
   }
   slot {class=OverlaySlot, properties {horizontalAlignment=HAlign_Center, verticalAlignment=VAlign_Center}}
  }
  Overlay_Notification_Text {
   widgetClassPath=&Overlay
   parent=GridPanel_77
   widget=Overlay_Notification_Text
   slot {class=GridSlot, properties {column=1, padding=<ltrb 10 0 10 0>}}
  }
  Text_Notification {
   widgetClassPath=&CommonTextBlock
   bIsVariable=true
   parent=Overlay_Notification_Text
   widget=Text_Notification
   properties {
    justification=Center
    mobileFontSizeMultiplier=0.8
    text=MIST RAVEN KIT COMPLETE
    wrapTextAt=70
    font {
     bForceMonospaced=false
     bMaterialIsStencil=false
     fontMaterial=null
     fontObject=&Burbank
     letterSpacing=0
     monospacedWidth=1
     size=10.5
     skewAmount=0
     typefaceFontName=Burbank-Black
     outlineSettings {
      bApplyOutlineToDropShadows=false
      bMiteredCorners=false
      bSeparateFillAlpha=false
      outlineMaterial=null
      outlineSize=0
      outlineColor=<rgba 0 0 0 1>
     }
    }
    shadowColorAndOpacity=<rgba 0 0 0 0.8>
    shadowOffset=<0 0>
    strikeBrush {
     bIsDynamicallyLoaded=false
     drawAs=Image
     imageType=FullColor
     mirroring=NoMirror
     resourceName=None
     resourceObject=null
     tiling=Horizontal
     imageSize=<128 128>
     margin=<ltrb 0 0 0 0>
     outlineSettings {
      bUseBrushTransparency=false
      roundingType=HalfHeightRadius
      width=0
      color {colorUseRule=UseColor_Specified, specifiedColor=<rgba 0 0 0 0>}
      cornerRadii=<0 0 0 0>
     }
     tintColor {colorUseRule=UseColor_Specified, specifiedColor=<rgba 1 1 1 1>}
     uVRegion {bIsValid=false, max=<0 0>, min=<0 0>}
    }
   }
   slot {class=OverlaySlot, properties {horizontalAlignment=HAlign_Center, verticalAlignment=VAlign_Center}}
  }
 ]
}
