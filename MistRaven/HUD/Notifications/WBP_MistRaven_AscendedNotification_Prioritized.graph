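# Prioritized widget wrapper for the MistRaven Ascended notification banner on the Fortnite HUD.
# It derives from AthenaCommonPrioritizedWidget and uses the Ascended notification banner as its root widget.
# One tick after Construct it flags itself as having active content through the prioritized widget interface, then clears that flag on Destruct.
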WidgetBlueprint /MistRaven/HUD/Notifications/WBP_MistRaven_AscendedNotification_Prioritized
graphs [EventGraph {nodes [
   Event Construct()
   > DelayUntilNextTick!(execute) -> (then)
   > SetHasActiveContent!(execute, self:Fort Prioritized Widget Interface, bNewActive:bool=true) -> (then)
   Event Destruct()
   > SetHasActiveContent(bNewActive=false)
  ]}]
variables []
widgets {
 info {
  inheritedWidgetCount=1
  parentClass=/Script/FortniteUI.AthenaCommonPrioritizedWidget
  rootWidgetClass=&/WBP_MistRaven_AscendedNotificationBanner.WBP_MistRaven_AscendedNotificationBanner_C
  widgetCount=2
 }
 widgets [
  WBP_MistRaven_AscendedNotificationBanner {
   widgetClassPath=&/WBP_MistRaven_AscendedNotificationBanner
   bIsVariable=true
   widget=WBP_MistRaven_AscendedNotificationBanner
  }
  TutorialHighlight {widgetClassPath=&TutorialHighlight}
 ]
}
